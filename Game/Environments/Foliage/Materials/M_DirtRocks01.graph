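# Default Lit material for dirt rock foliage, set up for instanced static meshes.
# World-space XY position divided by a tile parameter samples a small noise mask, which blends two color parameters and two specular parameters.
# A noise normal map on mesh UVs has its X and Y halved; the NormalTile multiply does not feed it.
# A clamped multi-lerp function takes per-instance random, and roughness is a scalar parameter.
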Material /Game/Environments/Foliage/Materials/M_DirtRocks01
graph {
 properties {
  bUsedWithInstancedStaticMeshes=true
  pixelDepthOffsetMode=PDOM_Legacy
  previewMesh=/Game/Environments/Foliage/Meshes/SM_DirtRocks_02
  usedShadingModels=Default Lit
 }
 nodes [
  Clamp(None=MaterialFunctionCall(A=PerInstanceRandom()) {materialFunction=&Lerp_Multiple_Float3}.Lerp 3 Inputs)
  n1 = WorldPosition()
  n2 = ScalarParameter() {defaultValue=300, parameterName=DiffTile}
  n3 = TextureSample(UVs=ComponentMask(None=Divide(A=@n1.XYZ, B=@n2)) {g=true, r=true}) {samplerType=SAMPLERTYPE_Masks, texture=/Game/Characters/Fortress_CharacterShader/SmallNoiseTexture}
  LinearInterpolate(
   A=VectorParameter() {parameterName=Color1, defaultValue=<rgba 0.255 0.2022 0.1379 1>}.RGB
   B=VectorParameter() {parameterName=Color2, defaultValue=<rgba 0.12 0.1121 0.102 1>}.RGB
   Alpha=@n3.R
  )
  LinearInterpolate(A=ScalarParameter() {defaultValue=0.2, parameterName=Spec}, B=ScalarParameter() {defaultValue=0.05, parameterName=Spec_1}, Alpha=@n3.R)
  n4 = TextureCoordinate()
  Multiply(A=@n4, B=ScalarParameter() {defaultValue=1, parameterName=NormalTile})
  Multiply(
   A=TextureSample(UVs=@n4) {desc=Local Rock Normal, samplerType=SAMPLERTYPE_Normal, texture=/Game/Global/Textures/Default/Noise/T_Noise_Normal_01}.RGB
   B=Constant3Vector() {constant=<rgba 0.5 0.5 1 0>}
  )
  PerInstanceRandom()
  ScalarParameter() {defaultValue=0.5, parameterName=Roughness}
 ]
}
