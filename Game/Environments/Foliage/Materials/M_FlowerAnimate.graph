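# M_FlowerAnimate is a two-sided, masked, Default Lit foliage material for animated flowers, enabled for instanced static meshes.
# The engine's SimpleGrassWind world position offset function supplies the sway, with constant intensity, weight, speed and additional offset inputs.
# The RGB of the T_FlowerTest texture is multiplied by a constant pale, slightly warm tint.
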Material /Game/Environments/Foliage/Materials/M_FlowerAnimate
graph {
 properties {bUsedWithInstancedStaticMeshes=true, blendMode=BLEND_Masked, pixelDepthOffsetMode=PDOM_Legacy, twoSided=true, usedShadingModels=Default Lit}
 nodes [
  MaterialFunctionCall!(WindIntensity=Constant() {r=0.25}, WindWeight=Constant() {r=0.25}, WindSpeed=Constant() {r=0.3}, AdditionalWPO=Constant() {r=0.2}) {materialFunction=/Engine/Functions/Engine_MaterialFunctions01/WorldPositionOffset/SimpleGrassWind}
  Multiply(
   A=TextureSample() {texture=/Game/Environments/Foliage/Textures/T_FlowerTest}.RGB
   B=Constant3Vector() {constant=<rgba 0.7396 0.6695 0.6622 0>}
  )
 ]
}
