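# M_FlowerTest is a two-sided masked material in the foliage folder, apparently a test material for flowers, enabled for instanced static meshes.
# It uses Default Lit shading, and its small graph samples the T_FlowerTest texture and darkens the RGB through a constant multiply.
# The graph shows no links from that result to material outputs, including Opacity Mask.
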Material /Game/Environments/Foliage/Materials/M_FlowerTest
graph {
 properties {bUsedWithInstancedStaticMeshes=true, blendMode=BLEND_Masked, pixelDepthOffsetMode=PDOM_Legacy, twoSided=true, usedShadingModels=Default Lit}
 nodes [Multiply(A=TextureSample() {texture=/Game/Environments/Foliage/Textures/T_FlowerTest}.RGB) {constB=0.2}]
}
