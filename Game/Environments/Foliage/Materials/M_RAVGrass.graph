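# M_RAVGrass is a two-sided masked material for instanced grass, colored by a world-projected, mip-biased ground texture to match the terrain.
# Panned noise drives wind rotation around Pivot Painter axes and perturbs normals.
# World-space height variation adds to position offset, faded out by vertex alpha.
# Opacity combines blade alpha with per-instance fade; tint and brightness branches are left unconnected.
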Material /Game/Environments/Foliage/Materials/M_RAVGrass
graph {
 properties {
  bFullyRough=true
  bTangentSpaceNormal=false
  bUseMaterialAttributes=true
  bUsedWithInstancedStaticMeshes=true
  blendMode=BLEND_Masked
  ditherOpacityMask=true
  pixelDepthOffsetMode=PDOM_Legacy
  translucencyPass=MTP_BeforeDOF
  twoSided=true
  usedShadingModels=Default Lit
 }
 nodes [
  Constant() {r=0.05}
  Constant() {r=1000}
  Constant() {r=0.3}
  Constant() {r=50}
  Constant() {r=-1}
  Constant() {r=5}
  Constant() {r=50}
  Constant() {r=1000}
  Constant() {r=2.5}
  n1 = ScalarParameter() {parameterName=VertexNormalWPO-Low}
  n2 = VertexNormalWS()
  n3 = ScalarParameter() {defaultValue=35, parameterName=VertexNormalWPO-High}
  n4 = TextureCoordinate() {coordinateIndex=1}
  n5 = Clamp(None=Power(Base=ComponentMask(None=@n4) {g=true}, Exp=ScalarParameter() {defaultValue=1, parameterName=VertexNormalWPO-Low_1}))
  LinearInterpolate(A=LinearInterpolate(A=Multiply(A=@n1, B=@n2), B=Multiply(A=@n3, B=@n2), Alpha=PerInstanceRandom()), B=@n2, Alpha=@n5)
  n6 = VectorParameter() {parameterName=TestNrm, defaultValue=<rgba -0.2 -0.2 1 1>}
  n7 = ComponentMask(None=WorldPosition() {worldPositionShaderOffset=WPT_ExcludeAllShaderOffsets}.XYZ) {g=true, r=true}
  n8 = Divide(A=@n7, B=ScalarParameter() {defaultValue=2000, group=WPO, parameterName=WindWaveScale})
  n9 = Time()
  n10 = Multiply(A=ScalarParameter() {defaultValue=0.075, group=WPO, parameterName=WindSpeed}, B=@n9)
  n11 = Sine(None=Multiply(A=@n9, B=ScalarParameter() {defaultValue=0.15, group=Lightning, parameterName=WindSlowDuration}))
  n12 = Multiply(
   A=TextureSample(UVs=Panner(Coordinate=@n8, Time=@n10) {speedX=0.4, speedY=-0.5}) {samplerType=SAMPLERTYPE_Normal, texture=/Game/Global/Textures/Default/Tileable_Patterns/T_StormCloudDetail01_N}.RGB
   B=LinearInterpolate(
    A=ScalarParameter() {defaultValue=0.15, group=WPO, parameterName=WindPower}
    B=ScalarParameter() {defaultValue=0.05, group=WPO, parameterName=WindPowerLow}
    Alpha=ConstantBiasScale(None=@n11)
   )
  )
  n13 = OneMinus(None=TextureSampleParameter2D() {parameterName=MainTexture_1, texture=/Game/Environments/Foliage/Textures/T_DesertGrass_01}.R)
  n14 = TextureCoordinate() {coordinateIndex=1}
  n15 = Clamp(None=LinearInterpolate(
    A=ScalarParameter() {parameterName=AAA111}
    B=ScalarParameter() {defaultValue=3, parameterName=AAA111_1}
    Alpha=ComponentMask(None=@n14) {g=true}
   ))
  n16 = TextureSampleParameter2D() {parameterName=MainTexture, texture=&T_DesertGrass_01}
  n17 = OneMinus(None=Multiply(A=@n16.A, B=PerInstanceFadeAmount()))
  n18 = ScalarParameter() {defaultValue=2, parameterName=DistanceOpacityMaskPower}
  n19 = ScalarParameter() {defaultValue=-5, group=WPO, parameterName=StaticWPOLow}
  n20 = ScalarParameter() {defaultValue=60, group=WPO, parameterName=StaticWPOHigh}
  n21 = TextureSample(UVs=Divide(A=@n7, B=ScalarParameter() {defaultValue=1500, group=WPO, parameterName=HeightVarianceScale})) {samplerType=SAMPLERTYPE_LinearColor, texture=/Game/Environments/Foliage/Textures/T_FieldGrass_01_M}
  n22 = Multiply(A=@n5, B=Multiply(A=Constant3Vector() {constant=<rgba 0 0 1 0>}, B=LinearInterpolate(A=@n19, B=@n20, Alpha=@n21.R)))
  n23 = MaterialFunctionCall() {materialFunction=/Engine/Functions/Engine_MaterialFunctions02/PivotPainter/PivotPainter_PerObjectData}
  n24 = DotProduct(A=@n23.X-Axis Vector, B=Subtract(A=WorldPosition().XYZ, B=@n23.Pivot Position))
  n25 = ComponentMask(None=@n12) {g=true, r=true}
  n26 = Constant() {r=1}
  n27 = RotateAboutAxis(
   NormalizedRotationAxis=Normalize(VectorInput=AppendVector(A=@n25, B=@n26))
   RotationAngle=Constant() {r=0.5}
   PivotPoint=Subtract(A=WorldPosition().XYZ, B=Multiply(A=@n23.X-Axis Vector, B=@n24))
   Position=Add(A=@n22, B=WorldPosition() {worldPositionShaderOffset=WPT_ExcludeAllShaderOffsets}.XYZ)
  )
  n28 = ScalarParameter() {defaultValue=1024, group=ModifiedGrass, parameterName=TextureSize_XY}
  n29 = Divide(A=WorldPosition() {worldPositionShaderOffset=WPT_ExcludeAllShaderOffsets}.XYZ, B=Multiply(A=Abs(None=@n28)) {constB=-1})
  n30 = TextureSampleParameter2D!(
   UVs=Multiply(A=ComponentMask(None=@n29) {g=true, r=true}, B=ScalarParameter() {defaultValue=2.25, group=Floor_1, parameterName=Scale_Floor})
   Bias=ScalarParameter() {defaultValue=4, group=ModifiedGrass, parameterName=MatchingTextureMipBias}
   Apply View MipBias
  ) {mipValueMode=TMVM_MipBias, parameterName=GrassTextureToMatch, texture=/Game/Environments/Hexmap/Zones/Spooky/Forest/T_Spooky_Grass_RAV}
  MakeMaterialAttributes!(
   BaseColor=@n30.RGB
   Metallic
   Specular=Constant() {r=0.25}
   Roughness=Constant() {r=1}
   Anisotropy
   EmissiveColor
   Opacity
   OpacityMask=OneMinus(None=Power(Base=@n17, Exp=@n18))
   Normal=LinearInterpolate(A=VertexNormalWS(), B=Multiply(A=@n12, B=@n6.RGB), Alpha=Multiply(A=@n13, B=@n15))
   Tangent
   WorldPositionOffset=LinearInterpolate(A=Add(A=@n22, B=@n27), B=Constant(), Alpha=VertexColor().A)
   SubsurfaceColor
   ClearCoat
   ClearCoatRoughness
   AmbientOcclusion
   Refraction
   CustomizedUVs_0
   CustomizedUVs_1
   CustomizedUVs_2
   CustomizedUVs_3
   CustomizedUVs_4
   CustomizedUVs_5
   CustomizedUVs_6
   CustomizedUVs_7
   PixelDepthOffset
   ShadingModel
   Displacement
  )
  Multiply(A=@n16.A, B=Power(Base=@n30.A, Exp=ScalarParameter() {defaultValue=1, parameterName=MaskOutStuffPush}))
  n31 = Multiply(A=PerInstanceRandom(), B=ScalarParameter() {defaultValue=3, parameterName=pppp1}) {constB=2}
  n32 = ScalarParameter() {defaultValue=2.3, parameterName=pppp1_1}
  n33 = LinearInterpolate(A=@n21.G, Alpha=Clamp(None=Subtract(A=@n31, B=@n32)))
  Multiply(A=ScalarParameter() {defaultValue=1, parameterName=DisregardWithZero}, B=@n33)
  n34 = Desaturation(None=@n30.RGB, Fraction=ScalarParameter() {defaultValue=0.5, group=ModifiedGrass, parameterName=GrassColorMaskSize_5})
  n35 = Multiply(A=@n34, B=VectorParameter() {parameterName=LightAreaColor_1, defaultValue=<rgba 0.8698 0.4847 0.7702 1>}.RGB)
  n36 = Multiply(A=@n34, B=VectorParameter() {parameterName=LightAreaColor, defaultValue=<rgba 0.8762 0.0325 0.8906 1>}.RGB)
  n37 = Desaturation(
   None=Multiply(A=@n30.RGB, B=ScalarParameter() {defaultValue=2.3, group=ModifiedGrass, parameterName=GrassColorBrightMultiply})
   Fraction=ScalarParameter() {defaultValue=0.2, group=ModifiedGrass, parameterName=GrassColorBrightSaturation}
  )
  n38 = Multiply(A=@n16.G, B=@n21.R)
  Reroute(None=LinearInterpolate(A=LinearInterpolate(A=@n30.RGB, B=@n37, Alpha=@n38), B=LinearInterpolate(A=@n35, B=@n36, Alpha=@n16.G), Alpha=@n33))
  WorldPosition()
 ]
}
